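; GA_Mash_MoveSpeedRandomize is a gameplay ability in the Mash playlist that gives its owner a slightly randomized movement speed.
; On activation it builds an outgoing GE_Mash_MoveSpeed effect spec and sets the SetByCaller.MoveSpeed magnitude to a random float in a narrow range around one.
; It then applies that spec to the owner, with no EndAbility call in the graph.
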(event Ability|EventActivateAbility
  (bind _returnvalue (Ability|GameplayEffect|AssignTagSetbyCallerMagnitude (Ability|MakeOutgoingGameplayEffectSpec "/Game/Athena/Playlists/Mash/GameAbilities/GE_Mash_MoveSpeed.GE_Mash_MoveSpeed_C") "(TagName=\"SetByCaller.MoveSpeed\")" (Math|Random|RandomFloatinRange 0.9 1.1)))
  (Ability|ApplyGameplayEffectSpecToOwner _returnvalue))
